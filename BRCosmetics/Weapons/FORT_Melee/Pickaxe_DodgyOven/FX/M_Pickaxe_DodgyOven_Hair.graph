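# Masked, two-sided Default Lit hair material for Niagara sprites in the DodgyOven pickaxe cosmetic's FX.
# It blends two near-black color parameters using a smoothstepped blue channel from a hair mask texture.
# It also samples a hair normal map, with both textures borrowed from the Mouse emote's FX assets.
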Material /BRCosmetics/Weapons/FORT_Melee/Pickaxe_DodgyOven/FX/M_Pickaxe_DodgyOven_Hair
graph {
 properties {bUsedWithNiagaraSprites=true, blendMode=BLEND_Masked, pixelDepthOffsetMode=PDOM_Legacy, twoSided=true, usedShadingModels=Default Lit}
 nodes [
  Constant()
  LinearInterpolate(
   A=VectorParameter() {parameterName=Color1, defaultValue=<rgba 0.03 0.03 0.03 1>}.RGB
   B=VectorParameter() {parameterName=Color2, defaultValue=<rgba 0.045 0.045 0.05 1>}.RGB
   Alpha=SmoothStep(Value=TextureSample() {texture=/BRCosmetics/Animation/Game/MainPlayer/Emotes/Mouse/FX/Textures/T_Mouse_Hair_E_M}.B) {constMax=0.3}
  )
  TextureSample() {samplerType=SAMPLERTYPE_Normal, texture=/BRCosmetics/Animation/Game/MainPlayer/Emotes/Mouse/FX/Textures/T_Mouse_Hair_E_N}
 ]
}
