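# Anim Blueprint for the DodgyOven pickaxe that procedurally swings its hanging bell and two chain bones with chained AnimDynamics nodes.
# The cached GameBehaviour pose pitches the chain master bone, then simulates root-relative with spherical limits on the bell.
# Nested bool blends choose between it and a mostly world-space variant; other simulation chains are unconnected.
# The EventGraph caches a character BodyMesh on customization load and sets bIsNotHoldingAxe when it is invalid.
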AnimBlueprint /BRCosmetics/Weapons/FORT_Melee/Pickaxe_DodgyOven/Meshes/DodgyOven_Axe_AnimBP
target_skeleton=&/DodgyOven_Axe_Skeleton
graphs [
 AnimGraph {nodes [
   Anim Dynamics - Bone: dyn_bell_chain_01!(ComponentPose, bAlphaBoolEnabled:bool=True, ExternalForce="0,0,0") -> (Pose) {
    alphaInputType=Bool
    angularDampingOverride=0.99
    angularSpringConstant=1000
    bOverrideAngularDamping=true
    bUseGravityOverride=true
    bUsePlanarLimit=false
    linearDampingOverride=0.9999
    simulationSpace=RootRelative
    alphaBoolBlend {blendInTime=0.1, blendOption=Linear, blendOutTime=0.1}
    boundBone {boneName=dyn_bell_chain_01}
    componentAppliedLinearAccClamp=<700 700 700>
    componentLinearAccScale=<0.4 0.4 0.4>
    componentLinearVelScale=<0.6 0.6 0.6>
    gravityOverride {z=-800}
    physicsBodyDefinitions [{
      boundBone {boneName=dyn_bell_chain_01}
      boxExtents=<7 7 7>
      constraintSetup {angularLimitsMax=<0 130 0>, angularLimitsMin=<0 5 0>, angularTarget=<0 0 0>, linearAxesMin=<-0.5 0 0>}
      localJointOffset=<1 0 0>
     }]
   }
   n1 = Anim Dynamics - Bone: dyn_bell_chain_01(bAlphaBoolEnabled=True, ExternalForce="0,0,0") {
    alphaInputType=Bool
    angularDampingOverride=0.99
    angularSpringConstant=1000
    bAngularSpring=true
    bOverrideAngularDamping=true
    bUseGravityOverride=true
    bUsePlanarLimit=false
    linearDampingOverride=0.9999
    simulationSpace=RootRelative
    alphaBoolBlend {blendInTime=0.1, blendOption=Linear, blendOutTime=0.1}
    boundBone {boneName=dyn_bell_chain_01}
    componentAppliedLinearAccClamp=<500 500 500>
    componentLinearAccScale=<0.4 0.4 0.4>
    componentLinearVelScale=<0.6 0.6 0.6>
    gravityOverride {z=-800}
    physicsBodyDefinitions [{
      boundBone {boneName=dyn_bell_chain_01}
      boxExtents=<7 7 7>
      constraintSetup {angularLimitsMax=<0 130 0>, angularLimitsMin=<0 5 0>, angularTarget=<0 0 0>, linearAxesMin=<-0.5 0 0>}
      localJointOffset=<1 0 0>
     }]
   }
   n2 = Anim Dynamics - Bone: dyn_bell_chain_02!(ComponentPose=@n1, bAlphaBoolEnabled:bool=True, ExternalForce="0,0,0") -> (Pose) {
    alphaInputType=Bool
    angularDampingOverride=0.99
    angularSpringConstant=1000
    bAngularSpring=true
    bOverrideAngularDamping=true
    bUseGravityOverride=true
    bUsePlanarLimit=false
    linearDampingOverride=0.9999
    simulationSpace=RootRelative
    alphaBoolBlend {blendInTime=0.1, blendOption=Linear, blendOutTime=0.1}
    boundBone {boneName=dyn_bell_chain_02}
    componentAppliedLinearAccClamp=<500 500 500>
    componentLinearAccScale=<0.4 0.4 0.4>
    componentLinearVelScale=<0.6 0.6 0.6>
    gravityOverride {z=-800}
    physicsBodyDefinitions [{
      boundBone {boneName=dyn_bell_chain_02}
      boxExtents=<6 6 7>
      constraintSetup {angularLimitsMax=<5 5 25>, angularLimitsMin=<-5 0 0>, angularTarget=<0 0 0>}
      localJointOffset=<1 0 0>
     }]
    sphericalLimits [{limitRadius=4, drivingBone {boneName=Melee_TwoHanded_Handle}, sphereLocalOffset=<5 0 8>}]
   }
   Anim Dynamics - Bone: dyn_bell!(ComponentPose=@n2, bAlphaBoolEnabled:bool=True, ExternalForce="0,0,0") -> (Pose) {
    alphaInputType=Bool
    angularDampingOverride=0.99
    angularSpringConstant=1000
    bAngularSpring=true
    bOverrideAngularDamping=true
    bUseGravityOverride=true
    bUsePlanarLimit=false
    bUseSphericalLimits=true
    linearDampingOverride=0.9999
    simulationSpace=RootRelative
    alphaBoolBlend {blendInTime=0.1, blendOption=Linear, blendOutTime=0.1}
    boundBone {boneName=dyn_bell}
    componentAppliedLinearAccClamp=<500 500 500>
    componentLinearAccScale=<0.4 0.4 0.4>
    componentLinearVelScale=<0.6 0.6 0.6>
    gravityOverride {z=-800}
    physicsBodyDefinitions [{
      boundBone {boneName=dyn_bell}
      boxExtents=<6 6 6>
      constraintSetup {angularLimitsMax=<0 15 15>, angularLimitsMin=<0 0 -15>, angularTarget=<0 0 0>}
      localJointOffset=<1 0 0>
     }]
    sphericalLimits [
     {limitRadius=4, drivingBone {boneName=None}, sphereLocalOffset=<5 0 8>}
     {limitRadius=4.6, drivingBone {boneName=Melee_TwoHanded_Handle}, sphereLocalOffset=<7 4 8>}
     {limitRadius=4.6, drivingBone {boneName=Melee_TwoHanded_Handle}, sphereLocalOffset=<7 -4 8>}
    ]
   }
   n3 = Anim Dynamics - Bone: dyn_bell_chain_02(bAlphaBoolEnabled=True, ExternalForce="0,0,0") {
    alphaInputType=Bool
    angularDampingOverride=0.99
    angularSpringConstant=1000
    bAngularSpring=true
    bOverrideAngularDamping=true
    bUseGravityOverride=true
    bUsePlanarLimit=false
    linearDampingOverride=0.9999
    simulationSpace=RootRelative
    alphaBoolBlend {blendInTime=0.1, blendOption=Linear, blendOutTime=0.1}
    boundBone {boneName=dyn_bell_chain_02}
    componentAppliedLinearAccClamp=<700 700 700>
    componentLinearAccScale=<0.4 0.4 0.4>
    componentLinearVelScale=<0.6 0.6 0.6>
    gravityOverride {z=-800}
    physicsBodyDefinitions [{
      boundBone {boneName=dyn_bell_chain_02}
      boxExtents=<4 4 5>
      constraintSetup {angularLimitsMax=<5 5 25>, angularLimitsMin=<-5 0 0>, angularTarget=<0 0 0>}
      localJointOffset=<1 0 0>
     }]
    sphericalLimits [{limitRadius=4, drivingBone {boneName=Melee_TwoHanded_Handle}, sphereLocalOffset=<5 0 8>}]
   }
   Anim Dynamics - Bone: dyn_bell(ComponentPose=@n3, bAlphaBoolEnabled=True, ExternalForce="0,0,0") {
    alphaInputType=Bool
    angularDampingOverride=0.99
    angularSpringConstant=1000
    bAngularSpring=true
    bOverrideAngularDamping=true
    bUseGravityOverride=true
    bUsePlanarLimit=false
    bUseSphericalLimits=true
    linearDampingOverride=0.9999
    simulationSpace=RootRelative
    alphaBoolBlend {blendInTime=0.1, blendOption=Linear, blendOutTime=0.1}
    boundBone {boneName=dyn_bell}
    componentAppliedLinearAccClamp=<700 700 700>
    componentLinearAccScale=<0.4 0.4 0.4>
    componentLinearVelScale=<0.6 0.6 0.6>
    gravityOverride {z=-800}
    physicsBodyDefinitions [{
      boundBone {boneName=dyn_bell}
      boxExtents=<5 5 5>
      constraintSetup {angularLimitsMax=<0 15 15>, angularLimitsMin=<0 0 -15>, angularTarget=<0 0 0>}
      localJointOffset=<1 0 0>
     }]
    sphericalLimits [
     {limitRadius=4, drivingBone {boneName=None}, sphereLocalOffset=<5 0 8>}
     {limitRadius=4.6, drivingBone {boneName=Melee_TwoHanded_Handle}, sphereLocalOffset=<7 4 8>}
     {limitRadius=4.6, drivingBone {boneName=Melee_TwoHanded_Handle}, sphereLocalOffset=<7 -4 8>}
    ]
   }
   n4 = Anim Dynamics - Bone: dyn_bell_chain_01(ComponentPose=Local To Component(), bAlphaBoolEnabled=True, ExternalForce="0,0,0") {
    alphaInputType=Bool
    angularDampingOverride=1
    angularSpringConstant=10000
    bUseGravityOverride=true
    bUsePlanarLimit=false
    linearDampingOverride=0.9999
    simulationSpace=RootRelative
    alphaBoolBlend {blendInTime=0.1, blendOption=Linear, blendOutTime=0.1}
    boundBone {boneName=dyn_bell_chain_01}
    componentAppliedLinearAccClamp=<700 700 700>
    componentLinearAccScale=<0.4 0.4 0.4>
    componentLinearVelScale=<0.6 0.6 0.6>
    gravityOverride {z=-980}
    physicsBodyDefinitions [{
      boundBone {boneName=dyn_bell_chain_01}
      boxExtents=<6 8 8>
      constraintSetup {angularLimitsMax=<0 130 0>, angularLimitsMin=<0 10 0>, angularTarget=<0 0 0>, linearAxesMin=<-0.5 0 0>}
      localJointOffset=<1.5 0 0>
     }]
   }
   n5 = Anim Dynamics - Bone: dyn_bell_chain_02(ComponentPose=@n4, bAlphaBoolEnabled=True, ExternalForce="0,0,0") {
    alphaInputType=Bool
    angularDampingOverride=0.99
    angularSpringConstant=1000
    bAngularSpring=true
    bOverrideAngularDamping=true
    bUseGravityOverride=true
    bUsePlanarLimit=false
    linearDampingOverride=0.9999
    simulationSpace=RootRelative
    alphaBoolBlend {blendInTime=0.1, blendOption=Linear, blendOutTime=0.1}
    boundBone {boneName=dyn_bell_chain_02}
    componentAppliedLinearAccClamp=<700 700 700>
    componentLinearAccScale=<0.4 0.4 0.4>
    componentLinearVelScale=<0.6 0.6 0.6>
    gravityOverride {z=-800}
    physicsBodyDefinitions [{
      boundBone {boneName=dyn_bell_chain_02}
      boxExtents=<4 4 5>
      constraintSetup {angularLimitsMax=<5 5 25>, angularLimitsMin=<-5 0 0>, angularTarget=<0 0 0>}
      localJointOffset=<1 0 0>
     }]
    sphericalLimits [{limitRadius=4, drivingBone {boneName=Melee_TwoHanded_Handle}, sphereLocalOffset=<5 0 8>}]
   }
   n6 = Anim Dynamics - Bone: dyn_bell(ComponentPose=@n5, bAlphaBoolEnabled=True, ExternalForce="0,0,0") {
    alphaInputType=Bool
    angularDampingOverride=0.99
    angularSpringConstant=1000
    bAngularSpring=true
    bOverrideAngularDamping=true
    bUseGravityOverride=true
    bUsePlanarLimit=false
    bUseSphericalLimits=true
    linearDampingOverride=0.9999
    simulationSpace=RootRelative
    alphaBoolBlend {blendInTime=0.1, blendOption=Linear, blendOutTime=0.1}
    boundBone {boneName=dyn_bell}
    componentAppliedLinearAccClamp=<700 700 700>
    componentLinearAccScale=<0.4 0.4 0.4>
    componentLinearVelScale=<0.6 0.6 0.6>
    gravityOverride {z=-800}
    physicsBodyDefinitions [{
      boundBone {boneName=dyn_bell}
      boxExtents=<5 5 5>
      constraintSetup {angularLimitsMax=<0 15 15>, angularLimitsMin=<0 0 -15>, angularTarget=<0 0 0>}
      localJointOffset=<1 0 0>
     }]
    sphericalLimits [
     {limitRadius=4, drivingBone {boneName=None}, sphereLocalOffset=<5 0 8>}
     {limitRadius=4.6, drivingBone {boneName=Melee_TwoHanded_Handle}, sphereLocalOffset=<7 4 8>}
     {limitRadius=4.6, drivingBone {boneName=Melee_TwoHanded_Handle}, sphereLocalOffset=<7 -4 8>}
    ]
   }
   Component To Local(ComponentPose=@n6)
   n7 = Anim Dynamics - Bone: dyn_bell_chain_01(ComponentPose=Local To Component(LocalPose=Use cached pose 'null_pose'()), bAlphaBoolEnabled=True, ExternalForce="0,0,0") {
    alphaInputType=Bool
    angularDampingOverride=1
    angularSpringConstant=10000
    bUseGravityOverride=true
    bUsePlanarLimit=false
    linearDampingOverride=0.9999
    simulationSpace=World
    alphaBoolBlend {blendInTime=0.1, blendOption=Linear, blendOutTime=0.1}
    boundBone {boneName=dyn_bell_chain_01}
    componentAppliedLinearAccClamp=<700 700 700>
    componentLinearAccScale=<0.4 0.4 0.4>
    componentLinearVelScale=<0.6 0.6 0.6>
    gravityOverride {z=-980}
    physicsBodyDefinitions [{
      boundBone {boneName=dyn_bell_chain_01}
      boxExtents=<6 8 8>
      constraintSetup {angularLimitsMax=<0 20 10>, angularLimitsMin=<0 0 -10>, angularTarget=<0 0 0>}
      localJointOffset=<1.5 0 0>
     }]
   }
   n8 = Anim Dynamics - Bone: dyn_bell_chain_02(ComponentPose=@n7, bAlphaBoolEnabled=True, ExternalForce="0,0,0") {
    alphaInputType=Bool
    angularDampingOverride=0.99
    angularSpringConstant=1000
    bAngularSpring=true
    bOverrideAngularDamping=true
    bUseGravityOverride=true
    bUsePlanarLimit=false
    linearDampingOverride=0.9999
    simulationSpace=World
    alphaBoolBlend {blendInTime=0.1, blendOption=Linear, blendOutTime=0.1}
    boundBone {boneName=dyn_bell_chain_02}
    componentAppliedLinearAccClamp=<700 700 700>
    componentLinearAccScale=<0.4 0.4 0.4>
    componentLinearVelScale=<0.6 0.6 0.6>
    gravityOverride {z=-800}
    physicsBodyDefinitions [{
      boundBone {boneName=dyn_bell_chain_02}
      boxExtents=<4 4 5>
      constraintSetup {angularLimitsMax=<5 5 25>, angularLimitsMin=<-5 0 0>, angularTarget=<0 0 0>}
      localJointOffset=<1 0 0>
     }]
    sphericalLimits [{limitRadius=4, drivingBone {boneName=Melee_TwoHanded_Handle}, sphereLocalOffset=<5 0 8>}]
   }
   n9 = Anim Dynamics - Bone: dyn_bell(ComponentPose=@n8, bAlphaBoolEnabled=True, ExternalForce="0,0,0") {
    alphaInputType=Bool
    angularDampingOverride=0.99
    angularSpringConstant=1000
    bAngularSpring=true
    bOverrideAngularDamping=true
    bUseGravityOverride=true
    bUsePlanarLimit=false
    simulationSpace=RootRelative
    alphaBoolBlend {blendInTime=0.1, blendOption=Linear, blendOutTime=0.1}
    boundBone {boneName=dyn_bell}
    componentAppliedLinearAccClamp=<700 700 700>
    componentLinearAccScale=<0.4 0.4 0.4>
    componentLinearVelScale=<0.6 0.6 0.6>
    gravityOverride {z=-800}
    physicsBodyDefinitions [{
      boundBone {boneName=dyn_bell}
      boxExtents=<5 5 5>
      constraintSetup {angularLimitsMax=<0 15 15>, angularLimitsMin=<0 0 -15>, angularTarget=<0 0 0>}
      localJointOffset=<1 0 0>
     }]
   }
   n10 = Component To Local(ComponentPose=@n9)
   n11 = Use cached pose 'GameBehaviour'()
   Output Pose(Result=Blend Poses by bool(BlendPose_0=Blend Poses by bool(BlendPose_0=@n10, BlendPose_1=@n11) {blendType=Linear}, BlendPose_1=Use cached pose 'GameBehaviour'()) {blendType=Linear}) {layerGroup=None}
   AnimGraphNode_SubInput_0 = Input Pose()
   Save cached pose 'null_pose'(Pose=@AnimGraphNode_SubInput_0) {cacheName=null_pose}
   n12 = "Transform (Modify) Bone - Bone: dyn_bell_chain_MSTR"(ComponentPose=Local To Component(LocalPose=Use cached pose 'null_pose'()), Alpha=1, Translation="0,0,0", Rotation="-30,0,0", Scale="1,1,1") {
    rotationMode=BMM_Additive
    rotationSpace=BCS_BoneSpace
    alphaBoolBlend {blendOption=Linear}
    boneToModify {boneName=dyn_bell_chain_MSTR}
    rotation {pitch=-30}
   }
   n13 = Anim Dynamics - Bone: dyn_bell_chain_01(ComponentPose=@n12, bAlphaBoolEnabled=True, ExternalForce="0,0,0") {
    alphaInputType=Bool
    angularDampingOverride=1
    angularSpringConstant=10000
    bUseGravityOverride=true
    bUsePlanarLimit=false
    linearDampingOverride=0.9999
    simulationSpace=RootRelative
    alphaBoolBlend {blendInTime=0.1, blendOption=Linear, blendOutTime=0.1}
    boundBone {boneName=dyn_bell_chain_01}
    componentAppliedLinearAccClamp=<700 700 700>
    componentLinearAccScale=<0.4 0.4 0.4>
    componentLinearVelScale=<0.6 0.6 0.6>
    gravityOverride {z=-980}
    physicsBodyDefinitions [{
      boundBone {boneName=dyn_bell_chain_01}
      boxExtents=<6 8 8>
      constraintSetup {angularLimitsMax=<0 90 0>, angularLimitsMin=<0 -10 0>, angularTarget=<0 0 0>, linearAxesMin=<-0.5 0 0>}
      localJointOffset=<1.5 0 0>
     }]
   }
   n14 = Anim Dynamics - Bone: dyn_bell_chain_02(ComponentPose=@n13, bAlphaBoolEnabled=True, ExternalForce="0,0,0") {
    alphaInputType=Bool
    angularDampingOverride=0.99
    angularSpringConstant=1000
    bAngularSpring=true
    bOverrideAngularDamping=true
    bUseGravityOverride=true
    bUsePlanarLimit=false
    linearDampingOverride=0.9999
    simulationSpace=RootRelative
    alphaBoolBlend {blendInTime=0.1, blendOption=Linear, blendOutTime=0.1}
    boundBone {boneName=dyn_bell_chain_02}
    componentAppliedLinearAccClamp=<700 700 700>
    componentLinearAccScale=<0.4 0.4 0.4>
    componentLinearVelScale=<0.6 0.6 0.6>
    gravityOverride {z=-800}
    physicsBodyDefinitions [{
      boundBone {boneName=dyn_bell_chain_02}
      boxExtents=<4 4 5>
      constraintSetup {angularLimitsMax=<5 5 25>, angularLimitsMin=<-5 0 0>, angularTarget=<0 0 0>}
      localJointOffset=<1 0 0>
     }]
    sphericalLimits [{limitRadius=4, drivingBone {boneName=Melee_TwoHanded_Handle}, sphereLocalOffset=<5 0 8>}]
   }
   n15 = Anim Dynamics - Bone: dyn_bell(ComponentPose=@n14, bAlphaBoolEnabled=True, ExternalForce="0,0,0") {
    alphaInputType=Bool
    angularDampingOverride=0.99
    angularSpringConstant=1000
    bAngularSpring=true
    bOverrideAngularDamping=true
    bUseGravityOverride=true
    bUsePlanarLimit=false
    bUseSphericalLimits=true
    linearDampingOverride=0.9999
    simulationSpace=RootRelative
    alphaBoolBlend {blendInTime=0.1, blendOption=Linear, blendOutTime=0.1}
    boundBone {boneName=dyn_bell}
    componentAppliedLinearAccClamp=<700 700 700>
    componentLinearAccScale=<0.4 0.4 0.4>
    componentLinearVelScale=<0.6 0.6 0.6>
    gravityOverride {z=-800}
    physicsBodyDefinitions [{
      boundBone {boneName=dyn_bell}
      boxExtents=<5 5 5>
      constraintSetup {angularLimitsMax=<0 15 15>, angularLimitsMin=<0 0 -15>, angularTarget=<0 0 0>}
      localJointOffset=<1 0 0>
     }]
    sphericalLimits [
     {limitRadius=4, drivingBone {boneName=None}, sphereLocalOffset=<5 0 8>}
     {limitRadius=4.6, drivingBone {boneName=Melee_TwoHanded_Handle}, sphereLocalOffset=<7 4 8>}
     {limitRadius=4.6, drivingBone {boneName=Melee_TwoHanded_Handle}, sphereLocalOffset=<7 -4 8>}
    ]
   }
   Save cached pose 'GameBehaviour'(Pose=Component To Local(ComponentPose=@n15)) {cacheName=GameBehaviour}
  ]}
 EventGraph {nodes [
   Event BlueprintUpdateAnimation()
   n1 = Get BodyMesh() -> (BodyMesh:Skeletal Mesh Component Budgeted Object Reference)
   > Set bIsNotHoldingAxe(bIsNotHoldingAxe:bool=NOT Boolean(A=IsValid(Object=@n1))) -> (Output_Get:bool)
   n2 = Event OnCharacterCustomizationLoaded()
   > Set BodyMesh(BodyMesh:Skeletal Mesh Component Budgeted Object Reference=Get Array Item with EFort Custom Part Type(Array=@n2.CustomMeshParts)) -> (Output_Get:Skeletal Mesh Component Budgeted Object Reference)
  ]}
]
variables [
 bIsNotHoldingAxe {type=boolean}
 BodyMesh {type=SkeletalMeshComponentBudgeted}
]
